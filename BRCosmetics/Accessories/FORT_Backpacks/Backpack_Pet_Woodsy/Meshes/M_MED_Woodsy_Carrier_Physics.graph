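# Physics asset for the carrier mesh of the Woodsy pet backpack cosmetic, giving a two-bone cup chain collision shapes and joint limits.
# A single box covers the root, and a capsule sits on each of dyn_cup_01 and dyn_cup_02.
# Constraints link dyn_cup_01 to root and dyn_cup_02 to dyn_cup_01, with matching tight swing and twist limits that restrict the cup to small rotations.
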PhysicsAsset /BRCosmetics/Accessories/FORT_Backpacks/Backpack_Pet_Woodsy/Meshes/M_MED_Woodsy_Carrier_Physics
bodies {
 dyn_cup_01 [{shapeType=Capsule, length=2.1058, radius=1.4034, center=<1.5865 -0.0082 -0.0028>, rotation=<rot 90 0 0>}]
 dyn_cup_02 [{shapeType=Capsule, length=4.58, radius=4.232, center=<4.1395 0.1604 -1.0406>, rotation=<rot 90 0 0>}]
 root [{shapeType=Box, extentX=28.461, extentY=25.9158, extentZ=29.8822, center=<0.0923 -11.6439 -3.9528>}]
}
constraints [
 {bone1Name=dyn_cup_01, bone2Name=root, swing1LimitDegrees=10, swing2LimitDegrees=10, twistLimitDegrees=15}
 {bone1Name=dyn_cup_02, bone2Name=dyn_cup_01, swing1LimitDegrees=10, swing2LimitDegrees=10, twistLimitDegrees=15}
]
